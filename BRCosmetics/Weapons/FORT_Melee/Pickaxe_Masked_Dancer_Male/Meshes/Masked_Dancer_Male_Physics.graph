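# Physics asset for the Masked Dancer Male two-handed pickaxe cosmetic, setting up two dangling belt chains on the weapon.
# A long capsule covers the handle, and each belt has a small capsule base followed by four box segments.
# Each base joins the handle with free swing and twist, then the first segment gets wide swing limits.
# Later segments use tighter swing and narrow twist limits down each chain.
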PhysicsAsset /BRCosmetics/Weapons/FORT_Melee/Pickaxe_Masked_Dancer_Male/Meshes/Masked_Dancer_Male_Physics
bodies {
 Melee_TwoHanded_Handle [{shapeType=Capsule, length=102.5, radius=2.5, center=<-0.0841 -1.1441 23.509>}]
 dyn_belt_a_01 [{shapeType=Box, extentX=4, extentY=3.5, extentZ=1.5, center=<2 0 0>}]
 dyn_belt_a_02 [{shapeType=Box, extentX=4, extentY=3.5, extentZ=1.5, center=<2 0 0>}]
 dyn_belt_a_03 [{shapeType=Box, extentX=4, extentY=3.5, extentZ=1.5, center=<2 0 0>}]
 dyn_belt_a_04 [{shapeType=Box, extentX=4, extentY=3.5, extentZ=1.5, center=<2 0 0>}]
 dyn_belt_a_base [{shapeType=Capsule, length=0.5, radius=0.5}]
 dyn_belt_b_01 [{shapeType=Box, extentX=4, extentY=3.5, extentZ=1.5, center=<2 0 0>}]
 dyn_belt_b_02 [{shapeType=Box, extentX=4, extentY=3.5, extentZ=1.5, center=<2 0 0>}]
 dyn_belt_b_03 [{shapeType=Box, extentX=4, extentY=3.5, extentZ=1.5, center=<2 0 0>}]
 dyn_belt_b_04 [{shapeType=Box, extentX=3.5, extentY=3.5, extentZ=1.5, center=<2 0 0>}]
 dyn_belt_b_base [{shapeType=Capsule, length=0.5, radius=0.5}]
}
constraints [
 {
  bone1Name=dyn_belt_a_base
  bone2Name=Melee_TwoHanded_Handle
  swing1LimitDegrees=45
  swing1Motion=Free
  swing2LimitDegrees=45
  swing2Motion=Free
  twistLimitDegrees=45
  twistMotion=Free
 }
 {bone1Name=dyn_belt_a_01, bone2Name=dyn_belt_a_base, swing1LimitDegrees=90, swing2LimitDegrees=90, twistLimitDegrees=10}
 {bone1Name=dyn_belt_a_02, bone2Name=dyn_belt_a_01, swing1LimitDegrees=25, swing2LimitDegrees=25, twistLimitDegrees=8}
 {bone1Name=dyn_belt_a_03, bone2Name=dyn_belt_a_02, swing1LimitDegrees=25, swing2LimitDegrees=25, twistLimitDegrees=8}
 {bone1Name=dyn_belt_a_04, bone2Name=dyn_belt_a_03, swing1LimitDegrees=25, swing2LimitDegrees=25, twistLimitDegrees=8}
 {
  bone1Name=dyn_belt_b_base
  bone2Name=Melee_TwoHanded_Handle
  swing1LimitDegrees=45
  swing1Motion=Free
  swing2LimitDegrees=45
  swing2Motion=Free
  twistLimitDegrees=45
  twistMotion=Free
 }
 {bone1Name=dyn_belt_b_01, bone2Name=dyn_belt_b_base, swing1LimitDegrees=90, swing2LimitDegrees=90, twistLimitDegrees=10}
 {bone1Name=dyn_belt_b_02, bone2Name=dyn_belt_b_01, swing1LimitDegrees=25, swing2LimitDegrees=25, twistLimitDegrees=8}
 {bone1Name=dyn_belt_b_03, bone2Name=dyn_belt_b_02, swing1LimitDegrees=25, swing2LimitDegrees=25, twistLimitDegrees=8}
 {bone1Name=dyn_belt_b_04, bone2Name=dyn_belt_b_03, swing1LimitDegrees=25, swing2LimitDegrees=25, twistLimitDegrees=8}
]
